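# Niagara system for the HoneySting cosmetic companion's emote: a golden glow flash with bursts of pollen and an outward dust shockwave.
# Pollen specks and soft clouds add age-curved spawn rates to their bursts, drifting on curl noise as they fade.
# DustShockwaveParent launches a torus ring of particles that DustShockwave reads through attribute readers to spawn swirling trails.
# Front-end user bools switch glow and dust colors, and front-end modifiers adjust shockwave speed and swirl.
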NiagaraSystem /CosmeticCompanions/FX/Builtins/HoneySting/Emote/NS_Emote_HoneySting
dependencies {
 usedDataInterfaces [&NiagaraDataInterfaceCurve, &NiagaraDataInterfaceArrayFloat, &NiagaraDataInterfaceParticleRead]
 usedDynamicInputs [
  &Lerp_Float
  &MakeCustomFloatFromBool
  &SetBoolByIntComparison
  &Multiply_Vector
  &RandomBool
 ]
 usedModules [&SystemState]
}
emitters {
 CloudPollen {
  emitterUpdateScript [
   &EmitterState(Life Cycle Mode=::NewEnumerator1, Loop Behavior=::NewEnumerator1, Loop Duration=5.8, UseLoopDelay=false)
   &SpawnBurst_Instantaneous(Spawn Count=2, Use Spawn Probability=false, UseLoopCountLimit=false)
   &SpawnRate(
    SpawnRate=&Multiply_Float(
     A=8
     B=&FloatFromCurve(FloatCurve=DataInterface() {propertyValues {Curve {keys [<at 4 1>,<at 4.5 0.4>]}}}, CurveIndex=$Emitter.Age)
    )
    Use Spawn Probability=false
   )
  ]
  particleSpawnScript [
   &InitializeParticle(
    Color=<rgba 2 1.3142 0.0035 0.3>
    Lifetime Max=1.5
    Lifetime Mode=::NewEnumerator1
    Sprite Rotation Mode=::NewEnumerator2
    Sprite Size Mode=::NewEnumerator3
    Uniform Sprite Size=&RandomRangeFloat(Minimum=45, Maximum=70)
    UsePositionOffset=false
    Write Material Random=true
   )
   &SystemLocation(MaskSpawnGroup=false, Offset=<0 0 10>)
   &CameraOffset(Camera Offset Amount=10)
   &ShapeLocation(
    Cylinder Height=10
    Cylinder Height Midpoint=0
    Cylinder Radius=10
    MaskSpawnGroup=false
    Override Seed=false
    Shape Primitive=::NewEnumerator1
    Surface Only=false
    UseOwnerScale=true
   )
  ]
  particleUpdateScript [
   &ParticleState()
   &ScaleSpriteSize(Uniform Curve Sprite Scale=DataInterface() {propertyValues {Curve {keys [{tangentMode=RCTM_User, time=0, value=1}, {tangentMode=RCTM_User, time=1, value=1.5}]}}})
   &ScaleColor(
    Scale Alpha=&FloatFromCurve(FloatCurve=DataInterface() {propertyValues {Curve {
        defaultValue=1
        keys [
         {arriveTangent=9.0725, interpMode=RCIM_Cubic, leaveTangent=9.0725, tangentMode=RCTM_User, time=0, value=0}
         {arriveTangent=-2.6507, interpMode=RCIM_Cubic, leaveTangent=-2.6507, tangentMode=RCTM_User, time=0.191, value=0.9847}
         {arriveTangent=-0.0405, interpMode=RCIM_Cubic, leaveTangent=-0.0405, tangentMode=RCTM_User, time=1, value=0}
        ]
       }}})
    ScaleA=true
    ScaleRGB=true
   )
   &CurlNoiseForce(
    Bias Noise=false
    Noise Strength=&Multiply_Float(
     A=100
     B=&FloatFromCurve(FloatCurve=DataInterface() {propertyValues {Curve {keys [<at 1.4 1 RCIM_Cubic>,<at 2.2 0.2 RCIM_Cubic>]}}}, CurveIndex=$Emitter.Age)
    )
    Pan Noise=false
   )
   &SolveForcesAndVelocity(Clamp Velocity=false, Limit Acceleration=false)
  ]
  rendererClasses [&NiagaraSpriteRendererProperties]
  renderers [
   {rendererClass=&NiagaraSpriteRendererProperties}
   {rendererClass=&NiagaraSpriteRendererProperties, rendererIndex=1}
  ]
 }
 DustShockwave {
  emitterUpdateScript [
   &EmitterState()
   &SpawnParticlesFromOtherEmitter(
    Attribute Reader=DataInterface() {propertyValues {EmitterBinding {bindingMode=Other, emitterName=DustShockwaveParent}}}
    Maximum Age Spawn Threshold=10000
    Spawn Rate=1000
    Spawn Rate Per Particle Cap=80
    Use Maximum Age=false
    Use Minimum Age=false
   )
  ]
  particleSpawnScript [
   &InitializeParticle(
    Color=<rgba 0.0849 1 0.0015 1>
    Lifetime=&RandomRangeFloat(Minimum=1, Maximum=1.5)
    Mass=5
    Mass Mode=::NewEnumerator1
    Sprite Rotation Mode=::NewEnumerator2
    UsePositionOffset=false
    Write Material Random=true
   )
   &SampleParticlesFromOtherEmitter(
    Age Sampling:ENiagara_DirectReadSamplingMode=::NewEnumerator0
    Emitter Sampling Mode:ENiagara_EmitterLocSamplingMode=::NewEnumerator0
    Lifetime Sampling:ENiagara_DirectReadSamplingMode=::NewEnumerator0
    Maximum Sampled Particle Age=0.1
    Module.MaxSampledAgeCutoff=true
    Module.MinSampledAgeCutoff=false
    Particle ID Sampling=::NewEnumerator1
    Position Sampling:ENiagara_DirectReadSamplingMode=::NewEnumerator3
    Reject Particle By Sampled Age=true
    Velocity Application:ENiagara_DirectReadApplicationMode=::NewEnumerator0
    Velocity Sampling=::NewEnumerator0
   )
   &CameraOffset(Camera Offset Amount=25)
  ]
  particleUpdateScript [
   &ParticleState()
   &SpriteRotationRate(Rotation Rate=&Multiply_Float(
     A=&FloatFromCurve(
      FloatCurve=DataInterface() {propertyValues {Curve {keys [<at 0 1 RCIM_Cubic>, {arriveTangent=-1.0596, tangentMode=RCTM_User, time=1, value=0.2}]}}}
      Scale Curve=200
     )
     B=&RandomRangeFloat(Minimum=-1, Maximum=1)
    )) {enabled=false}
   /SetVariables(
    Particles.Color=&MultiplyLinearColor(
     Linear Color=&MakeCustomLinearColorFromBool(Bool=$User.bIsFrontEndPreview, True Color=<rgba 1.1146 0.7324 0.0019 0.2>, False Color=<rgba 2 1.3142 0.0035 0.3>)
     Scale Factor=&MakeVector4FromVectorAndFloat(
      "Vector (RGB)"=&MakeCustomVectorFromBool(Bool=$User.bIsFrontEndPreview, True Vector=<2.2 2.2 2.2>, False Vector=<1 1 1>)
      "Float (Alpha)"=2
     )
    )
    Particles.Scale:NiagaraFloat=&RandomRangeFloat(Minimum=0.2, Maximum=0.4)
   ) {isSetParameters=true}
   &SolveForcesAndVelocity(Clamp Velocity=false, Limit Acceleration=false)
   /Niagara/Modules/AttributeReader/UpdateParticlesFromOtherEmitter(
    Attribute Reader=DataInterface() {propertyValues {EmitterBinding {bindingMode=Other, emitterName=DustShockwaveParent}}}
    Velocity Sampling:ENiagara_DirectReadSamplingMode=::NewEnumerator0
   )
   &self:Swirl(
    DistFreq:NiagaraFloat=&FN_FrontEndFloatModifier(Gameplay=0.15, Front End=0)
    OpacityFadePower=3
    RotSpeed=1.65
    Scale:NiagaraFloat=&FN_FrontEndFloatModifier(Gameplay=55, Front End=45)
    ScaleLength=0.45
    ScaleUpByNormAge=0.5
   ) {module=Swirl}
  ]
  rendererClasses [&NiagaraSpriteRendererProperties]
  renderers [{rendererClass=&NiagaraSpriteRendererProperties}]
 }
 DustShockwaveParent {
  emitterUpdateScript [
   &EmitterState()
   &SpawnBurst_Instantaneous(Spawn Count=20, Use Spawn Probability=false, UseLoopCountLimit=false)
  ]
  particleSpawnScript [
   &InitializeParticle(
    Color=<rgba 1 0.0572 0 1>
    Lifetime=0.7
    Mass=5
    Mass Mode=::NewEnumerator1
    Sprite Rotation Mode=::NewEnumerator2
    Sprite Size Mode=::NewEnumerator1
    Uniform Sprite Size Max=6
    Uniform Sprite Size Min=6
    UsePositionOffset=false
    Write Material Random=true
   )
   &ShapeLocation(
    Handle Radius=0
    Large Radius=1
    MaskSpawnGroup=false
    Offset=<0 0 7>
    Offset Mode=::NewEnumerator0
    Shape Primitive=::NewEnumerator3
    Torus Distribution Mode:ENiagaraTorusDistributionMode=::NewEnumerator1
    U Position=&ReturnNormalizedExecIndex(Normalized Index Scale=0.968)
    UseOwnerScale=true
   )
   &ApplyInitialForces()
   &AddVelocity(
    Module.Use Falloff Distance=false
    Origin Offset=<0 0 7>
    Override Seed=false
    Velocity Mode=::NewEnumerator1
    Velocity Speed=&FN_FrontEndFloatModifier(
     Gameplay=&RandomRangeFloat(Minimum=180, Maximum=220)
     Front End=&RandomRangeFloat(Minimum=160, Maximum=200)
    )
   )
  ]
  particleUpdateScript [
   &ParticleState()
   &Drag(Drag=3.5, Rotational Drag=0, Use Linear Drag=true, Use Rotational Drag=true)
   &AccelerationForce(Acceleration=<0 0 1500>)
   &SolveForcesAndVelocity(Clamp Velocity=false, Limit Acceleration=false)
  ]
  rendererClasses [&NiagaraSpriteRendererProperties]
  renderers [{rendererClass=&NiagaraSpriteRendererProperties}]
 }
 Glow {
  emitterUpdateScript [
   &EmitterState()
   &SpawnBurst_Instantaneous(Use Spawn Probability=false, UseLoopCountLimit=false)
  ]
  particleSpawnScript [
   &InitializeParticle(
    Color=&MakeCustomLinearColorFromBool(Bool=$User.bIsFrontEnd, True Color=<rgba 11 7.7647 0.2588 0.01>, False Color=<rgba 15 9.163 1.3596 0.01>)
    Lifetime=0.3
    Sprite Size Mode=::NewEnumerator3
    Uniform Sprite Size=101.1211
    UsePositionOffset=true
    Write Material Random=true
   )
   &SystemLocation(MaskSpawnGroup=false, Offset=<0 0 5>)
   &CameraOffset(Camera Offset Amount=15)
  ]
  particleUpdateScript [
   &ParticleState()
   &ScaleSpriteSize(Uniform Curve Sprite Scale=DataInterface() {propertyValues {Curve {keys [
        {arriveTangent=4.1785, interpMode=RCIM_Cubic, leaveTangent=4.1785, tangentMode=RCTM_User, time=0, value=0}
        {arriveTangent=1.1179, interpMode=RCIM_Cubic, leaveTangent=1.1179, tangentMode=RCTM_User, time=0.2042, value=0.6228}
        {arriveTangent=0.0224, interpMode=RCIM_Cubic, leaveTangent=0.0224, tangentMode=RCTM_User, time=1, value=1}
       ]}}})
   &SpriteRotationRate(Rotation Rate=&Sine(Normalized Angle=$Emitter.Age, Period=2.4044, Scale=135.79, Bias=0)) {module=SpriteRotationRate001}
   &ScaleColor(
    Scale Alpha=&FloatFromCurve(FloatCurve=DataInterface() {propertyValues {Curve {keys [
         {tangentMode=RCTM_User, time=0, value=1}
         {arriveTangent=-0.0405, interpMode=RCIM_Cubic, leaveTangent=-0.0405, tangentMode=RCTM_User, time=0.4, value=1}
         {arriveTangent=-2.6152, interpMode=RCIM_Cubic, leaveTangent=-2.6152, tangentMode=RCTM_User, time=1, value=0}
        ]}}})
    ScaleA=true
    ScaleRGB=true
   )
   /SetVariables(
    Particles.GlowColor=&Multiply_LinearColorByFloat(
     Linear Color=&MakeCustomLinearColorFromBool(Bool=$User.bIsFrontEnd, True Color=<rgba 2 1.4931 0.3215 0.5>, False Color=<rgba 2 1.0083 0.3061 0.5>)
     Float=&FloatFromCurve(
      FloatCurve=DataInterface() {propertyValues {Curve {keys [
          {tangentMode=RCTM_User, time=0, value=1}
          {arriveTangent=-0.0405, interpMode=RCIM_Cubic, leaveTangent=-0.0405, tangentMode=RCTM_User, time=0.4, value=1}
          {arriveTangent=-2.6152, interpMode=RCIM_Cubic, leaveTangent=-2.6152, tangentMode=RCTM_User, time=1, value=0}
         ]}}}
      Scale Curve=1.2
     )
    )
    Particles.GlowSize=&Multiply_Vector2D_ByFloat(
     Float=&FloatFromCurve(FloatCurve=DataInterface() {propertyValues {Curve {keys [
          {arriveTangent=4.1785, interpMode=RCIM_Cubic, leaveTangent=4.1785, tangentMode=RCTM_User, time=0, value=0}
          {arriveTangent=1.1179, interpMode=RCIM_Cubic, leaveTangent=1.1179, tangentMode=RCTM_User, time=0.4, value=1.1}
          {arriveTangent=0.0224, interpMode=RCIM_Cubic, leaveTangent=0.0224, tangentMode=RCTM_User, time=1, value=1.5}
         ]}}})
     Vector2D=<60 60>
    )
    Particles.SunRayColor:LinearColor=&Multiply_LinearColorByFloat(
     Linear Color=&MakeCustomLinearColorFromBool(Bool=$User.bIsFrontEnd, True Color=<rgba 11 7.7647 0.2588 0.01>, False Color=<rgba 1 0.6109 0.0906 0.1>)
     Float=&FloatFromCurve(FloatCurve=DataInterface() {propertyValues {Curve {keys [
          {tangentMode=RCTM_User, time=0, value=1}
          {arriveTangent=-0.0405, interpMode=RCIM_Cubic, leaveTangent=-0.0405, tangentMode=RCTM_User, time=0.4, value=1}
          {arriveTangent=-2.6152, interpMode=RCIM_Cubic, leaveTangent=-2.6152, tangentMode=RCTM_User, time=1, value=0}
         ]}}})
    )
   ) {isSetParameters=true}
   &SolveForcesAndVelocity(Clamp Velocity=false, Limit Acceleration=false)
  ]
  rendererClasses [&NiagaraSpriteRendererProperties]
  renderers [
   {rendererClass=&NiagaraSpriteRendererProperties}
   {rendererClass=&NiagaraSpriteRendererProperties, rendererIndex=1}
   {rendererClass=&NiagaraSpriteRendererProperties, rendererIndex=2}
  ]
 }
 Pollen {
  emitterUpdateScript [
   &EmitterState()
   &SpawnBurst_Instantaneous(Spawn Count=25, Use Spawn Probability=false, UseLoopCountLimit=false)
   &SpawnRate(
    SpawnGroup=1
    SpawnRate=&Multiply_Float(
     A=3
     B=&FloatFromCurve(FloatCurve=DataInterface() {propertyValues {Curve {keys [<at 3.5 3.5 RCIM_Cubic>,<at 4.5 0 RCIM_Cubic>]}}}, CurveIndex=$Emitter.Age)
    )
    Use Spawn Probability=false
   )
  ]
  particleSpawnScript [
   &InitializeParticle(
    Color Maximum=<rgba 1 0.4483 0.0047 0.3>
    Color Minimum=<rgba 1 0.8129 0.0516 1>
    Color Mode=::NewEnumerator2
    Lifetime Max=2.5
    Lifetime Min=1.5
    Lifetime Mode=::NewEnumerator1
    Sprite Rotation Mode=::NewEnumerator2
    Sprite Size Mode=::NewEnumerator1
    Uniform Sprite Size Max=2
    Uniform Sprite Size Min=1
    UsePositionOffset=false
    Write Material Random=true
   )
   &ShapeLocation(
    Hemisphere Distribution=<1 0.5>
    MaskSpawnGroup=false
    Override Seed=false
    Sphere Radius=5
    Sphere Surface Distribution=1
    UseOwnerScale=true
   )
   &AddVelocity(
    Module.Use Falloff Distance=false
    Origin Offset=<0 0 -2>
    Override Seed=false
    Velocity Mode=::NewEnumerator1
    Velocity Speed=&MaskFloat(
     Mask Source=0
     Mask Parameter=?
     Mask=0
     Masked Float=&RandomRangeFloat(Minimum=100, Maximum=200)
     Passthrough Float=&RandomRangeFloat(Minimum=0, Maximum=0)
    )
   )
   &ApplyInitialForces()
  ]
  particleUpdateScript [
   &ParticleState()
   &SpriteRotationRate(Rotation Rate=&SelectFloatFromArray(
     Array Sampling Mode=::NewEnumerator0
     Direct Array Index=?
     Float Selection Array=DataInterface() {propertyValues {FloatData [225,200,175,-175,-200,-225]}}
     Random Seed=0
    )) {module=SpriteRotationRate001}
   &MeshRotationForce(
    Coordinate Space:ENiagaraMeshTransforms=::NewEnumerator3
    "Lever Radius (cm)"=70
    Rotation=&RandomRangeVector(Minimum=<3 3 7>, Maximum=<5 5 9>)
   )
   &PointForce(
    Force Strength=&MaskFloat(Mask Source=0, Mask Parameter=?, Mask=0, Masked Float=260, Passthrough Float=0)
    Module.Use Falloff=true
    Module.UseLinearForce=false
   )
   &GravityForce(Gravity=<0 0 30>)
   &CurlNoiseForce(
    Noise Frequency=25
    Noise Strength=&MaskFloat(
     Mask Source=0
     Mask Parameter=?
     Mask=0
     Masked Float=&RandomRangeFloat(Minimum=100, Maximum=170)
     Passthrough Float=&RandomRangeFloat(Minimum=50, Maximum=90)
    )
    Pan Noise=false
   )
   &ScaleSpriteSize(Scale Factor=&Vector2DFromFloat(Value=&FloatFromCurve(FloatCurve=DataInterface() {propertyValues {Curve {keys [<at -0.0046 1 RCIM_Cubic>, {arriveTangent=-0.9943, interpMode=RCIM_Cubic, leaveTangent=-0.9943, time=0.7001, value=0.872}, <at 1.0011 0>]}}})))
   &ScaleColor(
    Scale Alpha=&FloatFromCurve(FloatCurve=DataInterface() {propertyValues {Curve {keys [<at -0.0046 1.0113 RCIM_Cubic>,<at 1.0011 0.002 RCIM_Cubic>]}}})
    ScaleA=true
    ScaleRGB=true
   )
   &Drag(Drag=&RandomRangeFloat(Minimum=6, Maximum=10), Use Linear Drag=true, Use Rotational Drag=true)
   &SolveForcesAndVelocity(Clamp Velocity=false, Limit Acceleration=false)
  ]
  rendererClasses [&NiagaraSpriteRendererProperties]
  renderers [{rendererClass=&NiagaraSpriteRendererProperties}]
 }
}
summary {
 systemName=NS_Emote_HoneySting
 emitters [
  {emitterName=Glow, rendererClasses [&NiagaraSpriteRendererProperties]}
  {emitterName=Pollen, rendererClasses [&NiagaraSpriteRendererProperties]}
  {emitterName=CloudPollen, rendererClasses [&NiagaraSpriteRendererProperties]}
  {emitterName=DustShockwaveParent, rendererClasses [&NiagaraSpriteRendererProperties]}
  {emitterName=DustShockwave, rendererClasses [&NiagaraSpriteRendererProperties]}
 ]
 userVariables [
  User.bIsFrontEnd {type=NiagaraBool}
  User.bIsFrontEndPreview {type=NiagaraBool}
 ]
}
